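# Ice floor material for the IceHotel set, blending cracked ice with sparkly snow.
# World-projected cracks use bump offsets at several depths and blurred mips to fake depth, with Fresnel shading and noise variation.
# Inverted red vertex color drives height-lerped snow coverage and boosts ice glow, while snow mixes in flake highlights.
# Forward shading gets ice only, mobile gets flat attributes, and the result feeds the destructible damage function.
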Material /Game/Environments/Apollo/Sets/IceHotel/Materials/M_IceFloor
graph {
 properties {bUseMaterialAttributes=true, pixelDepthOffsetMode=PDOM_Legacy, usedShadingModels=Default Lit}
 nodes [
  n1 = OneMinus(None=Fresnel() {baseReflectFraction=0.2, exponent=3})
  n2 = ComponentMask(None=Divide(A=WorldPosition().XYZ, B=Constant() {r=2570})) {g=true, r=true}
  n3 = ScalarParameter() {defaultValue=1024, parameterName=ScaleIce}
  n4 = ComponentMask(None=Divide(A=WorldPosition().XYZ, B=@n3)) {g=true, r=true}
  n5 = TextureSample!(
   UVs=BumpOffset(Coordinate=@n4, Height=Constant() {r=0.24}) {heightRatio=-0.4, referencePlane=0}
   Tex
   Level=Constant() {r=6}
   Apply View MipBias
  ) {mipValueMode=TMVM_MipLevel, samplerType=SAMPLERTYPE_Masks, texture=/Game/Athena/WinterArt/Textures/Cracks_05_M}
  n6 = TextureSample(UVs=BumpOffset(Coordinate=@n4, Height=Constant() {r=0.1931}) {heightRatio=-0.4, referencePlane=0}, Level=Constant() {r=5}) {mipValueMode=TMVM_MipLevel, samplerType=SAMPLERTYPE_Masks, texture=&Cracks_05_M}
  n7 = Divide(A=WorldPosition().XYZ, B=Constant() {r=125})
  n8 = Multiply(
   A=TextureSample!(UVs=ComponentMask(None=@n7) {g=true, r=true}, Tex, Apply View MipBias) {samplerType=SAMPLERTYPE_Grayscale, texture=&LowResBlurredNoise}.R
   B=Constant() {r=0.0125}
  )
  n9 = Add(A=BumpOffset(Coordinate=@n4, Height=Constant() {r=0.1333}) {heightRatio=-0.4, referencePlane=0}, B=@n8)
  n10 = OneMinus(None=Multiply(
    A=Multiply(A=@n5.R, B=@n6.R)
    B=TextureSample(UVs=@n9) {samplerType=SAMPLERTYPE_Masks, texture=&Cracks_05_M}.R
   ))
  n11 = Constant() {r=0.1}
  n12 = Add(A=@n4, B=@n8)
  n13 = TextureSample(UVs=@n12) {samplerType=SAMPLERTYPE_Grayscale, texture=/Game/Athena/WinterArt/Textures/Cracks}
  n14 = Multiply(A=Constant() {r=0.25}, B=OneMinus(None=@n13.R))
  n15 = Multiply(
   A=Add(A=Multiply(A=@n10, B=@n11), B=@n14)
   B=LinearInterpolate(
    A=Constant() {r=0.5}
    B=Constant() {r=3.5794}
    Alpha=TextureSample(UVs=@n2) {samplerType=SAMPLERTYPE_Grayscale, texture=&LowResBlurredNoise}.R
   )
  )
  n16 = VectorParameter() {parameterName=Ice Diffuse, defaultValue=<rgba 0.5757 0.6202 0.6823 0>}
  n17 = WorldPosition()
  n18 = Constant() {r=1578.0953}
  n19 = TextureSample(UVs=ComponentMask(None=Divide(A=@n17.XYZ, B=@n18)) {g=true, r=true}) {samplerType=SAMPLERTYPE_Grayscale, texture=&LowResBlurredNoise}
  n20 = OneMinus(None=@n19.R)
  n21 = Constant() {r=0.2}
  n22 = TextureSample(UVs=@n12) {samplerType=SAMPLERTYPE_Masks, texture=&Cracks_05_M}
  n23 = Add(A=Multiply(A=OneMinus(None=@n22.R), B=Constant() {r=0.124}), B=@n14)
  n24 = VectorParameter() {parameterName="Emissive ", defaultValue=<rgba 0.0019 0.0358 0.1458 0>}
  n25 = TextureObject() {samplerType=SAMPLERTYPE_Normal, texture=/Game/Athena/WinterArt/Textures/Cracks_N}
  n26 = Multiply(A=@n22.R, B=ScalarParameter() {defaultValue=0.025, parameterName=Detail N Value})
  n27 = Divide(A=WorldPosition().XYZ, B=Divide(A=Constant() {r=1024}, B=ScalarParameter() {defaultValue=1, group=Tile, parameterName=TileTextureScale}))
  n28 = Multiply(
   A=Constant3Vector() {constant=<rgba 1 1 0 1>}
   B=TextureSample(UVs=ComponentMask(None=@n27) {g=true, r=true}) {samplerType=SAMPLERTYPE_Normal, texture=/Game/Global/Textures/Default/Tileable_Patterns/T_IceScratch1_N}.RGB
  )
  n29 = SetMaterialAttributes!(
   MaterialAttributes
   Base Color=LinearInterpolate(A=Multiply(A=@n1, B=@n15), B=@n16.RGB, Alpha=@n19.R)
   Specular=Multiply(A=Constant() {r=0.5}, B=@n13.R)
   Roughness=Add(A=Multiply(A=@n20, B=@n21), B=Multiply(A=@n1, B=@n23))
   Emissive Color=Multiply(A=@n24.RGB, B=@n1)
   Normal=Add(
    A=MaterialFunctionCall(TextureObject=@n25, TextureSize=@n3) {materialFunction=&WorldAlignedNormal}.Z Texture
    B=Multiply(A=@n26, B=@n28)
   )
  ) {attributeSetTypes [
    #1
    #2
    #3
    #4
    #5
   ]}
  n30 = GetMaterialAttributes(None=@n29) {attributeGetTypes [#4]}
  n31 = Multiply(A=@n30.Emissive Color, B=Constant() {r=3}) {constB=4}
  n32 = OneMinus(None=VertexColor().R)
  n33 = TextureSample!(UVs=@n12, Tex, Level=ScalarParameter() {defaultValue=5, parameterName=Ice Mip}, Apply View MipBias) {constMipValue=5, mipValueMode=TMVM_MipLevel, samplerType=SAMPLERTYPE_Grayscale, texture=&Cracks}
  n34 = Multiply(A=@n33.R, B=Constant() {r=1.5})
  n35 = Constant3Vector() {constant=<rgba -1 1 1 1>}
  n36 = ComponentMask(None=Divide(A=WorldPosition().XYZ, B=ScalarParameter() {defaultValue=250, group=Snow, parameterName=Scale Snow})) {g=true, r=true}
  n37 = TextureSample(UVs=@n36, Level=Constant() {r=4}) {mipValueMode=TMVM_MipLevel, samplerType=SAMPLERTYPE_Normal, texture=/Game/Packages/DS_Fortnight/ExperimentalTextures/Dirt_N}
  n38 = SetMaterialAttributes!(
   MaterialAttributes
   Base Color=VectorParameter() {group=Snow, parameterName=Snow Color, defaultValue=<rgba 0.541 0.5569 0.6146 0>}.RGB
   Specular=Constant() {r=0.4}
   Roughness=Constant() {r=0.75}
   Normal=LinearInterpolate(A=Multiply(A=@n35, B=@n37.RGB), B=Constant3Vector() {constant=<rgba 0 0 1 0>}, Alpha=Constant() {r=0.7}) {constAlpha=0.8}
  ) {attributeSetTypes [
    #1
    #2
    #3
    #5
   ]}
  n39 = WorldPosition()
  n40 = ScalarParameter() {defaultValue=54.3955, group=Snow, parameterName=Scale Flakes}
  n41 = TextureSampleParameter2D(UVs=ComponentMask(None=Divide(A=@n39.XYZ, B=@n40)) {g=true, r=true}) {parameterName=FlakeNormal, samplerType=SAMPLERTYPE_Normal, texture=/Game/Environments/Sets/Neo/Vehicles/Textures/T_Flakes_N}
  n42 = SetMaterialAttributes!(
   MaterialAttributes
   Normal=@n41.RGB
   Roughness=Constant()
   Specular=Constant() {r=1}
   Emissive Color=Constant3Vector() {constant=<rgba 0.022 0.0298 0.0469 0>}
   Base Color=Constant() {r=0.7}
  ) {attributeSetTypes [
    #5
    #3
    #2
    #4
    #1
   ]}
  n43 = Constant() {r=0.6548}
  n44 = TextureSample!(UVs=@n36, Tex, Apply View MipBias) {texture=/Game/Athena/Environments/Island/Textures/Dirt_D}
  n45 = Multiply(A=Multiply(A=Max(A=@n41.R, B=@n41.R), B=OneMinus(None=@n44.A)), B=Fresnel(Normal=VertexNormalWS()) {baseReflectFraction=0.25, exponent=2})
  n46 = Constant() {r=1}
  n47 = GetMaterialAttributes(None=BlendMaterialAttributes(
    A=@n38
    B=@n42
    Alpha=MaterialFunctionCall(Transition Phase=@n43, Height Texture=@n45, Contrast=@n46) {materialFunction=&HeightLerp}.Alpha
   ) {vertexAttributeBlendType=UseA}) {attributeGetTypes [#4]}
  n48 = VectorParameter() {group=Snow, parameterName=Snow Emissive Tint, defaultValue=<rgba 0.026 0.0483 0.092 0>}
  n49 = VertexColor()
  n50 = Multiply(A=OneMinus(None=@n22.R), B=Constant() {r=0.7333})
  n51 = BlendMaterialAttributes(
   A=SetMaterialAttributes!(
    MaterialAttributes=@n30.MaterialAttributes
    Emissive Color=MaterialFunctionCall(A=@n30.Emissive Color, B=@n31, Transition Phase=@n32, Height Texture=@n34) {materialFunction=&HeightLerp}.Results
   ) {attributeSetTypes [#4]}
   B=SetMaterialAttributes(MaterialAttributes=@n47.MaterialAttributes, Emissive Color=Add(A=@n47.Emissive Color, B=@n48.RGB)) {attributeSetTypes [#4]}
   Alpha=MaterialFunctionCall(
    Transition Phase=OneMinus(None=@n49.R)
    Height Texture=Max(A=@n44.A, B=@n50)
    Contrast=ScalarParameter() {defaultValue=0.7933, group=Snow, parameterName=Snow Contrast}
   ) {materialFunction=&HeightLerp}.Alpha
  ) {vertexAttributeBlendType=UseA}
  n52 = SetMaterialAttributes!(
   MaterialAttributes
   Base Color=@n16.RGB
   Specular=Constant() {r=0.3}
   Roughness=Constant() {r=0.2}
   Emissive Color=@n24.RGB
   Normal=Constant3Vector() {constant=<rgba 0 0 1 0>}
  ) {attributeSetTypes [
    #1
    #2
    #3
    #4
    #5
   ]}
  MaterialFunctionCall(Material_In=ShadingPathSwitch(Default=@n51, Deferred=@n51, Forward=@n29, Mobile=@n52), UseCrackedDamageTextures=StaticBool()) {materialFunction=&MF_FN_DestructibleDamage}
 ]
}
