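# Material function in the IceHotel set that splits a mesh between two Vector3 inputs, likely texture samples, using its second UV channel.
# Floor and Saturate on the U coordinate of UV index 1 form a hard 0/1 mask, so UVs at U of 1 or above take the second input.
# The static switch Use2ndMaterial, in group Material_2, enables this blend or passes the first input through alone.
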MaterialFunction /Game/Environments/Apollo/Sets/IceHotel/Materials/MF_UV2_2TexturesSplit
graph {nodes [
  n1 = FunctionInput() {inputName=1 V3}
  n2 = FunctionInput() {inputName=2 V3, sortPriority=1}
  n3 = TextureCoordinate() {coordinateIndex=1}
  n4 = Saturate(None=Floor(None=ComponentMask(None=@n3) {r=true}))
  FunctionOutput(None=StaticSwitchParameter(True=LinearInterpolate(A=@n1, B=@n2, Alpha=@n4), False=@n1) {group=Material_2, parameterName=Use2ndMaterial}) {outputName=V3}
 ]}
