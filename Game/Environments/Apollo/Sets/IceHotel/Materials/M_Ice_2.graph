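# Subsurface ice material for the Apollo Ice Hotel set, with an optional snow layer on upward-facing surfaces and destructible-damage vertex animation.
# Camera-offset noise fakes refraction and reflection in an emissive term, adding sun/sky scattering and scaling by time of day.
# A height gradient modulates glow, specular and rock normals, which blend with world-aligned scratches; an optional sphere-masked gradient tints glow.
# Forward and mobile paths use simplified ice and flat snow normals.
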Material /Game/Environments/Apollo/Sets/IceHotel/Materials/M_Ice_2
graph {
 properties {
  bUseMaterialAttributes=true
  bUsedWithInstancedStaticMeshes=true
  bUsedWithMeshDeformer=true
  bUsedWithNanite=true
  bUsedWithNiagaraMeshParticles=true
  bUsedWithSkeletalMesh=true
  pixelDepthOffsetMode=PDOM_Legacy
  shadingModel=MSM_Subsurface
  usedShadingModels=Subsurface
 }
 nodes [
  n1 = VectorParameter() {group=Colors, parameterName=IceColor, defaultValue=<rgba 0.0689 0.2289 0.4792 1>}
  n2 = MaterialFunctionCall(Base=@n1.RGB, Blend=VectorParameter() {group=ActorGradient, parameterName=Actor Glow, defaultValue=<rgba 0.8021 0 0.2803 0>}.RGB) {materialFunction=&Blend_Overlay}
  Reroute(None=StaticSwitchParameter(True=@n2.Result, False=@n1.RGB) {group=ActorGradient, parameterName=Use Actor Gradient})
  n3 = Constant3Vector() {constant=<rgba 0 0 1 0>}
  n4 = LinearInterpolate(
   A=MaterialFunctionCall(
    TextureObject=TextureObject() {samplerType=SAMPLERTYPE_Normal, texture=/Game/ContentCreationTools/TextureCreation/Textures/T_BPCreated_MacroNormal_01}
    TextureSize=ScalarParameter() {defaultValue=512, group=Snow, parameterName=SnowScale}
   ) {materialFunction=&WorldAlignedTexture}.Z Texture
   B=@n3
   Alpha=Constant() {r=0.7}
  ) {constAlpha=0.8}
  n5 = SetMaterialAttributes!(
   MaterialAttributes
   Base Color=VectorParameter() {group=Snow, parameterName=Snow Color, defaultValue=<rgba 0.776 0.776 0.776 0>}.RGB
   Specular=ScalarParameter() {defaultValue=0.5143, group=Snow, parameterName=SnowSpec}
   Roughness=Constant() {r=1}
   Normal=ShadingPathSwitch(Default=@n4, Deferred=@n4, Forward=@n3, Mobile=@n3)
   Subsurface Color=Constant()
  ) {attributeSetTypes [
    #1
    #2
    #3
    #4
    #5
   ]}
  n6 = Constant3Vector() {constant=<rgba 0 0 1 0>}
  n7 = WorldPosition()
  n8 = ActorPositionWS()
  n9 = MaterialProxyReplace(Realtime=ScalarParameter() {defaultValue=5376, group=WorldSpaceGradient, parameterName=WH Start}, MaterialProxy=Constant())
  n10 = ComponentMask(None=Subtract(A=@n7.XYZ, B=StaticSwitchParameter(True=@n8, False=@n9) {group=WorldSpaceGradient, parameterName=ActorPivotRamp})) {b=true}
  n11 = Saturate(None=Divide(A=@n10, B=ScalarParameter() {defaultValue=1200, group=WorldSpaceGradient, parameterName=WH Blur}))
  n12 = TextureSampleParameter2D!(UVs, Level, Apply View MipBias) {
   group=Textures
   mipValueMode=TMVM_MipLevel
   parameterName=N
   samplerSource=SSM_Wrap_WorldGroupSettings
   samplerType=SAMPLERTYPE_Normal
   texture=/Game/Environments/Props/Rocks/Textures/T_Corrupt_Rock_AB_N
  }
  n13 = TextureSampleParameter2D() {
   group=Textures
   mipValueMode=TMVM_MipLevel
   parameterName=N2
   samplerSource=SSM_Wrap_WorldGroupSettings
   samplerType=SAMPLERTYPE_Normal
   texture=&T_Corrupt_Rock_AB_N
  }
  n14 = MaterialFunctionCall!(1 V3=@n12.RGB, 2 V3=@n13.RGB) {materialFunction=&/MF_UV2_2TexturesSplit}
  n15 = StaticSwitchParameter(True=LinearInterpolate(A=@n14.V3, B=@n6, Alpha=@n11), False=@n14.V3) {defaultValue=true, group=WorldSpaceGradient, parameterName=Use World Gradient}
  n16 = Transform(None=@n15)
  n17 = Saturate(None=Multiply(A=ComponentMask(None=@n16) {b=true}, B=ScalarParameter() {defaultValue=1, group=Snow, parameterName=SnowUpVector}))
  n18 = Fresnel(Normal=VertexNormalWS()) {baseReflectFraction=0.5, exponent=1}
  n19 = OneMinus(None=@n18)
  n20 = MaterialProxyReplace(Realtime=Multiply(A=@n1.RGB, B=@n19), MaterialProxy=@n1.RGB)
  n21 = ScalarParameter() {defaultValue=0.55, parameterName=Spec}
  n22 = Max(A=Divide(A=@n10, B=Constant() {r=155}), B=ScalarParameter() {defaultValue=1})
  n23 = LinearInterpolate(
   A=Multiply(A=ScalarParameter() {defaultValue=0.4, group=WorldSpaceGradient, parameterName=WH Min}, B=Saturate(None=@n22))
   B=ScalarParameter() {defaultValue=1.5, group=WorldSpaceGradient, parameterName=WH Max}
   Alpha=@n11
  )
  n24 = Saturate(None=@n23)
  n25 = TextureSampleParameter2D!(UVs, Apply View MipBias) {
   group=Textures
   parameterName=D
   samplerSource=SSM_Wrap_WorldGroupSettings
   texture=/Game/Environments/Props/Rocks/Textures/T_Corrupt_Rock_AB_D
  }
  n26 = TextureSampleParameter2D() {
   group=Textures
   parameterName=D2
   samplerSource=SSM_Wrap_WorldGroupSettings
   texture=&T_Corrupt_Rock_AB_D
  }
  n27 = ComponentMask(None=MaterialFunctionCall(1 V3=@n25.A, 2 V3=@n26.A) {materialFunction=&MF_UV2_2TexturesSplit}.V3) {r=true}
  n28 = SphereMask(
   A=Subtract(A=@n7.XYZ, B=ActorPositionWS())
   B=VectorParameter() {group=ActorGradient, parameterName=PivotLocation}.RGB
   Radius=Multiply(A=ObjectRadius(), B=ScalarParameter() {defaultValue=1, group=ActorGradient, parameterName=Sphere Radius, sortPriority=1})
   Hardness=ScalarParameter() {group=ActorGradient, parameterName=SphereHardness, sortPriority=2}
  )
  n29 = VertexNormalWS()
  n30 = CameraVectorWS()
  n31 = Saturate(None=Power(Base=DotProduct(A=@n29, B=@n30), Exp=ScalarParameter() {defaultValue=2, group=ActorGradient, parameterName=Fresnel Power}))
  n32 = Saturate(None=Multiply(A=@n27, B=Multiply(A=@n28, B=@n31)))
  n33 = Desaturation(
   None=StaticSwitchParameter(True=LinearInterpolate(A=@n1.RGB, B=@n2.Result, Alpha=@n32), False=@n1.RGB) {group=ActorGradient, parameterName=Use Actor Gradient}
   Fraction=LinearInterpolate(A=Constant() {r=-0.6}, B=Constant() {r=-0.2}, Alpha=@n24)
  )
  n34 = MaterialFunctionCall(
   Value_Morning=ScalarParameter() {defaultValue=0.3, group=TOD, parameterName=Morning}
   Value_Day=ScalarParameter() {defaultValue=0.8, group=TOD, parameterName=Day}
   Value_Evening=ScalarParameter() {defaultValue=0.3, group=TOD, parameterName=Evening}
   Value_Night=ScalarParameter() {defaultValue=0.45, group=TOD, parameterName=Night}
  ) {materialFunction=&Scalar_Time_of_Day}
  n35 = ComponentMask(None=WorldPosition().XYZ) {g=true, r=true}
  n36 = ScalarParameter() {defaultValue=512, group=Reflections, parameterName=RefractionScale}
  n37 = ComponentMask(None=CameraVectorWS()) {g=true, r=true}
  n38 = ScalarParameter() {defaultValue=0.5, group=Reflections, parameterName=RefractionOffsetSpeed}
  n39 = TextureSampleParameter2D(UVs=Add(A=Divide(A=@n35, B=@n36), B=Multiply(A=@n37, B=@n38))) {
   group=Reflections
   parameterName=Caustics
   samplerSource=SSM_Wrap_WorldGroupSettings
   samplerType=SAMPLERTYPE_Grayscale
   texture=&LowResBlurredNoise
  }
  n40 = Constant() {r=4}
  n41 = TextureSampleParameter2D!(UVs, Level=@n40, Apply View MipBias) {
   group=Textures
   mipValueMode=TMVM_MipLevel
   parameterName=D
   samplerSource=SSM_Wrap_WorldGroupSettings
   texture=&T_Corrupt_Rock_AB_D
  }
  n42 = TextureSampleParameter2D(Level=@n40) {
   group=Textures
   mipValueMode=TMVM_MipLevel
   parameterName=D2
   samplerSource=SSM_Wrap_WorldGroupSettings
   texture=&T_Corrupt_Rock_AB_D
  }
  n43 = MaterialFunctionCall(1 V3=@n41.A, 2 V3=@n42.A) {materialFunction=&MF_UV2_2TexturesSplit}
  n44 = CameraVectorWS()
  n45 = ComponentMask(None=@n44) {g=true, r=true}
  n46 = Constant() {r=0.5}
  n47 = TextureSampleParameter2D!(UVs=Add(A=ComponentMask(None=@n43.V3) {r=true}, B=Multiply(A=@n45, B=@n46)), Apply View MipBias) {
   group=Reflections
   parameterName=Caustics
   samplerSource=SSM_Wrap_WorldGroupSettings
   samplerType=SAMPLERTYPE_Grayscale
   texture=&LowResBlurredNoise
  }
  n48 = VertexNormalWS()
  n49 = Multiply(
   A=Multiply(A=@n47.R, B=ScalarParameter() {defaultValue=1.5, group=Reflections, parameterName=Fake Reflection Value})
   B=Saturate(None=DotProduct(A=@n44, B=@n48))
  )
  n50 = Multiply(A=LinearInterpolate(Alpha=@n11) {constA=1, constB=0}, B=@n49)
  n51 = VertexNormalWS()
  n52 = LinearInterpolate(
   B=Add(A=@n39.R, B=StaticSwitchParameter(True=@n50, False=@n49) {defaultValue=true, group=WorldSpaceGradient, parameterName=Use World Gradient})
   Alpha=OneMinus(None=Fresnel(Normal=@n51) {baseReflectFraction=0.2, exponent=1})
  ) {constA=1}
  n53 = Constant3Vector() {constant=<rgba 0 0 1 0>}
  n54 = CameraVectorWS()
  n55 = Clamp(None=DotProduct(A=Multiply(A=@n53) {constB=-1}, B=@n54))
  n56 = ScalarParameter() {defaultValue=3, group=Lighting, parameterName=SkyExponent}
  n57 = AtmosphericLightVector()
  n58 = Clamp(None=DotProduct(A=Multiply(A=@n57) {constB=-1}, B=@n54))
  n59 = ScalarParameter() {defaultValue=12, group=Lighting, parameterName=SunExponent}
  n60 = Add(
   A=Multiply(A=Power(Base=@n55, Exp=@n56), B=ScalarParameter() {defaultValue=1.5, group=Lighting, parameterName=SkyValue})
   B=Multiply(A=Power(Base=@n58, Exp=@n59), B=ScalarParameter() {defaultValue=0.2, group=Lighting, parameterName=SunValue})
  )
  n61 = MaterialFunctionCall!(World Position) {materialFunction=/Engine/Functions/Engine_MaterialFunctions02/Utility/DepthFromWorldPosition}
  n62 = ScalarParameter() {defaultValue=512, group=Lighting, parameterName=DepthClose}
  n63 = Saturate(None=Power(Base=Divide(A=@n61.Result, B=@n62), Exp=ScalarParameter() {defaultValue=3, group=Lighting, parameterName=Depth Exponent}))
  n64 = StaticSwitchParameter(True=Add(A=@n52, B=Multiply(A=@n60, B=@n63)), False=@n52) {defaultValue=true, group=Lighting, parameterName="Sun&&Sky"}
  n65 = StaticSwitchParameter(True=Multiply(A=@n23, B=@n64), False=@n64) {defaultValue=true, group=WorldSpaceGradient, parameterName=Use World Gradient}
  n66 = ComponentMask(None=VertexNormalWS()) {b=true}
  n67 = TextureObjectParameter() {group=Tile, parameterName=R_Tile, samplerType=SAMPLERTYPE_Masks, texture=/Game/Global/Textures/Default/Tileable_Patterns/T_IceScratch1_M}
  n68 = TextureObjectParameter() {group=Tile, parameterName=N_Tile, samplerType=SAMPLERTYPE_Normal, texture=/Game/Global/Textures/Default/Tileable_Patterns/T_IceScratch1_N}
  n69 = MaterialFunctionCall!(
   Diffuse Texture Object=@n67
   Spec Texture Object
   TextureSize=Divide(A=Constant() {r=1024}, B=ScalarParameter() {defaultValue=0.75, group=Tile, parameterName=TileTextureScale})
   --------------------------
   Use seperate settings for Z projection
   Use seperate texture for Z projection
   Texture Size Z projection
   Z Plane Diffuse Texture Object
   Z Plane Spec Texture Object
   "--------------------------      "
   ProjectionTransitionContrast
   Use Texture Alpha Heightmap to Refine the Transistion
   HeightLerpTransitionContrast
   PickTheDiffuseHeightChannel
   TheDiffuseHeightChannel
   "--------------------------       "
   NormalTextureObject=@n68
   World Space Normals=StaticBool()
   "--------------------------  "
   Export Float 4 Diffuse
   Export Float 4 Spec
   "--------------------------     "
   WorldPosition
   World Space Normal
  ) {materialFunction=/Engine/Functions/Engine_MaterialFunctions02/Texturing/WorldAlignedTextures_Complex}
  n70 = LinearInterpolate(A=@n69.Normal XY Texture, B=@n6, Alpha=Multiply(A=@n19, B=Abs(None=@n66)))
  n71 = ScalarParameter() {defaultValue=0.7, group=Tile, parameterName=Tile N Flatten}
  n72 = Multiply(A=@n27) {constB=0.5}
  n73 = Multiply(A=@n18, B=@n69.XY Diffuse Texture)
  n74 = SetMaterialAttributes!(
   MaterialAttributes
   Base Color=@n20
   Specular=LinearInterpolate(A=ScalarParameter() {defaultValue=0.35, parameterName=Spec Bottom}, B=@n21, Alpha=Saturate(None=@n24))
   Roughness=Multiply(A=Add(A=@n73, B=@n72), B=ScalarParameter() {defaultValue=0.1, parameterName=R})
   Emissive Color=Multiply(A=@n33, B=Multiply(A=@n65, B=@n34.Result))
   Normal=MaterialFunctionCall(
    BaseNormal=@n15
    AdditionalNormal=MaterialFunctionCall(Normal=@n70, Flatness=@n71) {materialFunction=&FlattenNormal}.Result
   ) {materialFunction=&BlendAngleCorrectedNormals}.Result
   Subsurface Color=@n1.RGB
  ) {attributeSetTypes [
    #1
    #2
    #3
    #6
    #4
    #5
   ]}
  n75 = MakeMaterialAttributes!(
   BaseColor=@n20
   Metallic
   Specular=@n21
   Roughness=Constant() {r=0.2}
   Anisotropy
   EmissiveColor=Multiply(A=@n33, B=Multiply(A=@n52, B=@n34.Result))
   Opacity
   OpacityMask
   Normal=@n14.V3
   Tangent
   WorldPositionOffset
   SubsurfaceColor
   ClearCoat
   ClearCoatRoughness
   AmbientOcclusion
   Refraction
   CustomizedUVs_0
   CustomizedUVs_1
   CustomizedUVs_2
   CustomizedUVs_3
   CustomizedUVs_4
   CustomizedUVs_5
   CustomizedUVs_6
   CustomizedUVs_7
   PixelDepthOffset
   ShadingModel
   Displacement
  )
  n76 = ShadingPathSwitch(Default=@n74, Deferred=@n74, Forward=@n75, Mobile=@n75)
  SetMaterialAttributes!(
   MaterialAttributes=StaticSwitchParameter(
    True=MaterialFunctionCall!(Base Material=@n76, Top Material=@n5, Alpha=@n17) {materialFunction=/Engine/Functions/MaterialLayerFunctions/MatLayerBlend_Standard}.Blended Material
    False=@n76
   ) {defaultValue=true, parameterName=Snow}
   World Position Offset=MaterialFunctionCall(
    UseDestructibleDamage=StaticBoolParameter() {defaultValue=true, group=z_code, parameterName=Code_UseHitEffects}
    DestructibleDamageAmount=ScalarParameter() {group=z_code, parameterName=Code_DestructionAmount}
   ) {materialFunction=&MF_FN_DestructibleDamage}.DamageAnimWPO
  ) {attributeSetTypes [#7]}
 ]
}
