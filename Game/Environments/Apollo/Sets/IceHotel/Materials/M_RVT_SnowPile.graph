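# Nanite-enabled snow pile material for the Apollo Ice Hotel set that blends plain off-white snow with landscape-driven snow.
# Samples the Apollo landscape runtime virtual texture and passes it through the snow progression function, with a material proxy fallback to plain snow.
# Vertex alpha drives a HeightLerp against a world-XY projected snow texture, breaking up the blend.
# An RVT switch function selects plain snow when RVT is unavailable.
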Material /Game/Environments/Apollo/Sets/IceHotel/Materials/M_RVT_SnowPile
graph {
 properties {
  bTangentSpaceNormal=false
  bUseMaterialAttributes=true
  bUsedWithNanite=true
  pixelDepthOffsetMode=PDOM_Legacy
  usedShadingModels=Default Lit
 }
 nodes [
  n1 = MakeMaterialAttributes!(
   BaseColor=Constant3Vector() {constant=<rgba 0.7 0.7 0.7 0>}
   Metallic
   Specular
   Roughness=Constant() {r=1}
   Anisotropy
   EmissiveColor
   Opacity
   OpacityMask
   Normal
   Tangent
   WorldPositionOffset
   SubsurfaceColor
   ClearCoat
   ClearCoatRoughness
   AmbientOcclusion
   Refraction
   CustomizedUVs_0
   CustomizedUVs_1
   CustomizedUVs_2
   CustomizedUVs_3
   CustomizedUVs_4
   CustomizedUVs_5
   CustomizedUVs_6
   CustomizedUVs_7
   PixelDepthOffset
   ShadingModel
   Displacement
  )
  n2 = RuntimeVirtualTextureSample!(UVs, World Position, Mip Level) {
   materialType=BaseColor_Normal_Specular_YCoCg
   mipValueMode=RVTMVM_MipLevel
   virtualTexture=/Game/Athena/Apollo/Maps/Landscape/VT/RVT_Apollo_Landscape
  }
  n3 = MaterialFunctionCall!(
   IsStaticSnow
   UseAlternateStaticSnowMask
   AlternateStaticSnowMask
   "UseRoad&&OcclusionMask"
   IsForGrassMeshToPushDown
   AdditionalMaskToRemoveSnow
   IsFallingSnowFX
   UseMinimapCoords
   IncreaseSnowUsingSlope
   UseSmoothBorder
   LerpHighM
   TemporalHeightLerpContrast
   IsLobbyIsland
   UseBR-IslandMask
   IsTree
   IsForRenderTarget
   UseSparkle
   UseSnowNormals
   UseAdditionalMaskToRemoveSnow
   IsTangentSpaceInput
   UseVertexNormalsForBasicSnow
   UseTangentSnowNormals
   Material=MakeMaterialAttributes(BaseColor=@n2.BaseColor, Specular=@n2.BaseColor, Roughness=@n2.BaseColor, Normal=@n2.BaseColor)
  ) {materialFunction=/Game/Athena/Apollo/Environments/Landscape/Materials/Biome/UniqueBiomeMaterialFunctions/MF_Apollo_Snow_Progression}
  n4 = VertexColor()
  n5 = WorldPosition() {worldPositionShaderOffset=WPT_ExcludeAllShaderOffsets}
  n6 = TextureSample(UVs=Multiply(A=ComponentMask(None=@n5.XYZ) {g=true, r=true}, B=Constant() {r=0.002})) {
   samplerSource=SSM_Wrap_WorldGroupSettings
   samplerType=SAMPLERTYPE_LinearColor
   texture=/Game/Athena/Environments/Landscape/Textures/T_Athena_Snow_01_M
  }
  n7 = Constant() {r=0.15}
  MaterialFunctionCall(
   NoRVT=@n1
   RVT=BlendMaterialAttributes(
    A=@n1
    B=MaterialProxyReplace(Realtime=@n3.Material, MaterialProxy=@n1)
    Alpha=MaterialFunctionCall(Transition Phase=@n4.A, Height Texture=@n6.G, Contrast=@n7) {materialFunction=&HeightLerp}.Alpha
   ) {vertexAttributeBlendType=UseA}
  ) {materialFunction=&FN_RVT_Switch_Material}
 ]
}
